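; BrickSoup tree controller that switches holiday lights and damageability on Asteria trees as named events start and end.
; A TreeEvents array pairs tree references with event names, and GetEventNames returns the unique non-empty names.
; The event active and ended handlers call UpdateTreeState, which finds valid trees matching that name.
; Active events set EnableLights and make those trees undamageable, while ended events reverse both.
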(event EventBeginPlay
  (|CallParentFunction))

(fn GetEventNames
  (for _array_element (Variables|Default|GetTreeEvents)
    (if (not (Utilities|String|IsEmpty (.EventReference_6_DA93745C43AEFB8AB3AAB9874AE24EF4 (Utilities|Struct|BreakSTRUCTBrickSoupTreeEvent _array_element))))
      (Utilities|Array|AddUnique (LocalVariables|GetLocal%20Event%20Names) (.EventReference_6_DA93745C43AEFB8AB3AAB9874AE24EF4 (Utilities|Struct|BreakSTRUCTBrickSoupTreeEvent _array_element)))))
  (return (LocalVariables|GetLocal%20Event%20Names)))

(fn HandleEventActive
  (params (EventName :type String))
  (CallFunction|UpdateTreeState EventName true))

(fn HandleEventEnded
  (params (EventName :type String))
  (CallFunction|UpdateTreeState EventName))

(fn UpdateTreeState
  (params (EventName :type String) (Active :type Boolean))
  (bind _eventstate (LocalVariables|GetEventState))
  (LocalVariables|SetEventState Active)
  (for _array_element (Variables|Default|GetTreeEvents)
    (if (Utilities|String|EqualExactly(String) (.EventReference_6_DA93745C43AEFB8AB3AAB9874AE24EF4 (Utilities|Struct|BreakSTRUCTBrickSoupTreeEvent _array_element)) EventName)
      (Utilities|IsValid (Utilities|Struct|BreakSTRUCTBrickSoupTreeEvent _array_element)
        (:Is%20Valid
          (Variables|/Game/Environments/Asteria/Foliage/Trees/Blueprint_Parent/BP_Tree_Asteria_Parent_HolidayScript.BP_Tree_Asteria_Parent_HolidayScript_C|SetEnableLights _eventstate (Utilities|Struct|BreakSTRUCTBrickSoupTreeEvent _array_element))
          (Variables|/Script/Engine.Actor|SetbCanBeDamaged (not _eventstate) (Utilities|Struct|BreakSTRUCTBrickSoupTreeEvent _array_element)))))))

(fn ConstructionScript)
