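; Cheat manager for BrickSoupSpoon that switches the world between Day 1, Day 2 and Day 3 stages using World Partition data layers, moving to Day 3 on BeginPlay.
; Each MoveToDay event passes its layer to UnloadStageandLoadNewDataLayer, which handles unloading CurrentStage, storing the new layer and activating it through the DataLayerManager.
; RemoveBricksoup unloads the current stage without clearing the variable.
; A standalone LoadNewDataLayer function activates a given layer but is not called here.
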(event EventBeginPlay
  (CallFunction|MovetoDay3))

(event Custom|MoveToDay1
  (CallFunction|UnloadStageAndLoadNewDataLayer (Variables|Default|GetDay1DataLayer)))

(event Custom|MoveToDay3
  (CallFunction|UnloadStageAndLoadNewDataLayer (Variables|Default|GetDay3DataLayer)))

(event Custom|MoveToDay2
  (CallFunction|UnloadStageAndLoadNewDataLayer (Variables|Default|GetDay2DataLayer)))

(event Custom|RemoveBricksoup
  (bind _currentstage (Variables|Default|GetCurrentStage))
  (Utilities|IsValid _currentstage
    (:Is%20Valid
      (bind _returnvalue (WorldPartition|GetDataLayerManager))
      (DataLayers|SetDataLayerRuntimeState _returnvalue _currentstage "Unloaded" false :!class /Script/Engine.DataLayerManager))))

(fn LoadNewDataLayer
  (params (NewDataLayer :type Data%20Layer%20Asset%20Object%20Reference))
  (bind _returnvalue (WorldPartition|GetDataLayerManager))
  (Utilities|IsValid _returnvalue
    (:Is%20Valid
      (DataLayers|SetDataLayerRuntimeState _returnvalue NewDataLayer "Activated" false :!class /Script/Engine.DataLayerManager))))

(fn UnloadStageandLoadNewDataLayer
  (params (NewDatalayer :type Data%20Layer%20Asset%20Object%20Reference))
  (bind _currentstage (Variables|Default|GetCurrentStage))
  (Utilities|IsValid _currentstage
    (:Is%20Valid
      (bind _returnvalue (WorldPartition|GetDataLayerManager))
      (Utilities|IsValid _returnvalue
        (:Is%20Valid
          (DataLayers|SetDataLayerRuntimeState _returnvalue _currentstage "Unloaded" false :!class /Script/Engine.DataLayerManager))))
    (:Is%20Not%20Valid
      _))
  (bind _output_get (Variables|Default|SetCurrentStage NewDatalayer))
  (DataLayers|SetDataLayerRuntimeState _returnvalue _output_get "Activated" false :!class /Script/Engine.DataLayerManager))

(fn ConstructionScript)
