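; Base Blueprint for a Brick Soup Spoon chunk prop that drops loot when it dies.
; On BeginPlay it binds a handler to its OnDied event.
; The handler prints a debug string, spawns loot at the actor's location from its tier group loot table and loot level, then stops the LoopingIdle audio component through a StopSound event.
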(event EventBeginPlay
  (EventDispatchers|BindEventtoOnDied (delegate Custom|OnDied_Event)))

(event Custom|OnDied_Event
  (params (DamagedActor :type Actor%20Object%20Reference) (Damage :type Float%20%28single-precision%29) (InstigatedBy :type Controller%20Object%20Reference) (DamageCauser :type Actor%20Object%20Reference) (HitLocation :type Vector) (FHitComponent :type Primitive%20Component%20Object%20Reference) (BoneName :type Name) (Momentum :type Vector))
  (Development|PrintString "Spawn Chunk Loot")
  (Class|AthenaFunctionLibrary|SpawnLootfromTableUsingItemEntry (Variables|Default|GetTier%20Group%20Name) (Loot|GetLootLevel) -1 (Transformation|GetActorLocation) "0, 0, 0" 0 true)
  (CallFunction|StopSound))

(event Custom|StopSound
  (Audio|Components|Audio|Stop (Variables|Default|GetLoopingIdle)))

(fn ConstructionScript)
