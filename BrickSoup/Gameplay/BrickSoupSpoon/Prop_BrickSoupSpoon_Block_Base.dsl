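; Construction script for the BrickSoupSpoon block base prop in the BrickSoup gameplay content.
; It assigns the configured idle sound from the Effects variables to the prop's idle audio component.
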(fn ConstructionScript
  (Audio|Components|Audio|SetSound (Variables|Default|GetIdleComponent) (Variables|Effects|GetIdleSound)))
